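# Material function for Underground skylight surfaces, replacing input material attributes with a two-sided foliage look when a static switch is on.
# Parallax-offset masks and a sphere mask blend tinted roof color, specular, roughness and opacity.
# Camera position ping-pongs between noise normals, which feed the surface normal and a skylight-cubemap sparkle reflection.
# Shading-path and quality switches vary emissive, subsurface and opacity, with emissive divided by eye adaptation.
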MaterialFunction /Game/Environments/Helios/Sets/Underground/Materials/MF_UndergroundSkylight
graph {nodes [
  n1 = TextureCoordinate()
  n2 = TextureSample(UVs=@n1) {
   samplerSource=SSM_Wrap_WorldGroupSettings
   samplerType=SAMPLERTYPE_Normal
   texture=/Game/Environments/Helios/Sets/Underground/Textures/T_Noise
  }
  n3 = TextureSample(UVs=@n1) {
   samplerSource=SSM_Wrap_WorldGroupSettings
   samplerType=SAMPLERTYPE_Normal
   texture=&T_Noise
  }
  n4 = Frac(None=Divide(A=CameraPositionWS(), B=ScalarParameter() {defaultValue=1024, group=PostRVTSnow, parameterName=Epic_SnowSparklePingPongDistance}) {constB=5000})
  n5 = Multiply(A=Abs(None=Add(A=@n4) {constB=-0.5})) {constB=2}
  n6 = LinearInterpolate(
   A=@n2.RGB
   B=MaterialFunctionCall(X=@n3.B, Y=@n3.B, Z=@n3.B) {materialFunction=&MakeFloat3}.Result
   Alpha=NamedRerouteDeclaration(None=Saturate(None=@n5)) {name=pingPongMask, nodeColor=<rgba 0 1 0.2969 1>}
  )
  n7 = Transform(None=@n6)
  Add(A=CameraPositionWS(), B=@n7)
  n8 = Power(Base=DotProduct(A=CameraVectorWS(), B=@n7)) {constExponent=0.25}
  n9 = TextureSample(UVs=BumpOffset(Coordinate=@n1, Height=Constant(), HeightRatioInput=LinearInterpolate() {constAlpha=1, constB=-0.035}) {heightRatio=-0.075}) {
   samplerSource=SSM_Wrap_WorldGroupSettings
   samplerType=SAMPLERTYPE_Masks
   texture=/Game/Environments/Helios/Sets/Underground/Textures/T_UndergroundSkylight_Masks
  }
  n10 = ScalarParameter() {parameterName=2}
  n11 = ScalarParameter() {defaultValue=0.075, parameterName=HeightRation}
  n12 = TextureSample!(UVs, Tex, Level, Apply View MipBias) {
   mipValueMode=TMVM_MipLevel
   samplerSource=SSM_Wrap_WorldGroupSettings
   samplerType=SAMPLERTYPE_Masks
   texture=&T_UndergroundSkylight_Masks
  }
  n13 = TextureSample!(
   UVs=BumpOffset(Coordinate=@n1, Height=Constant(), HeightRatioInput=LinearInterpolate(A=@n10, B=@n11, Alpha=@n12.G)) {heightRatio=-0.1}
   Tex
   Apply View MipBias
  ) {
   samplerSource=SSM_Wrap_WorldGroupSettings
   samplerType=SAMPLERTYPE_Masks
   texture=&T_UndergroundSkylight_Masks
  }
  n14 = Max(A=@n13.G, B=SphereMask(A=TextureCoordinate(), B=Constant2Vector() {g=1, r=1}) {attenuationRadius=0.5, hardnessPercent=0})
  n15 = LinearInterpolate(A=ScalarParameter() {defaultValue=0.25, parameterName=FrameDarken}, B=@n9.R, Alpha=Multiply(A=@n9.B, B=@n14))
  n16 = Subtract(A=@n8, B=Multiply(A=@n15, B=ScalarParameter() {defaultValue=1.5, parameterName=NoiseSterngth}))
  Ceil(None=@n16)
  n17 = Constant()
  n18 = Divide(A=NamedRerouteUsage(), B=EyeAdaptation().EyeAdaptation)
  n19 = TextureSampleParameter2D(UVs=@n1) {
   parameterName=BaseColor
   samplerSource=SSM_Wrap_WorldGroupSettings
   texture=/Game/Environments/Helios/Sets/Underground/Textures/T_Underground_Roof_B_B_E
  }
  n20 = Multiply(
   A=@n19.RGB
   B=LinearInterpolate(A=ScalarParameter() {parameterName=MinNoise}, B=ScalarParameter() {defaultValue=1, parameterName=MaxNoise}, Alpha=@n15)
  )
  n21 = Multiply(A=@n20)
  n22 = LinearInterpolate(A=@n21, B=Multiply(A=@n20, B=ScalarParameter() {defaultValue=2, parameterName=EmissiveStrength_Max}), Alpha=NamedRerouteUsage())
  n23 = EyeAdaptation()
  n24 = NamedRerouteUsage()
  n25 = LinearInterpolate(
   A=@n21
   B=Multiply(A=@n20, B=VectorParameter() {parameterName=SubsurfaceColorTint, defaultValue=<rgba 2 2 2 1>}.RGB)
   Alpha=NamedRerouteUsage()
  )
  n26 = ScalarParameter() {parameterName=MinOpacity}
  n27 = ScalarParameter() {defaultValue=1, parameterName=MaxOpacity}
  n28 = LinearInterpolate(A=@n26, B=@n27, Alpha=@n15)
  n29 = TextureSample(UVs=@n1) {
   samplerSource=SSM_Wrap_WorldGroupSettings
   samplerType=SAMPLERTYPE_Normal
   texture=/Game/Environments/Helios/Sets/Underground/Textures/T_Underground_Roof_B_N
  }
  n30 = Constant() {r=0.4}
  n31 = SetMaterialAttributes!(
   MaterialAttributes
   BaseColor=@n20
   Specular=LinearInterpolate(A=ScalarParameter() {defaultValue=0.5, parameterName=MinSpec}, B=ScalarParameter() {defaultValue=1, parameterName=MaxSpec}, Alpha=@n15)
   Roughness=LinearInterpolate(
    A=ScalarParameter() {defaultValue=0.1, parameterName=MinRoughness}
    B=ScalarParameter() {defaultValue=0.2, parameterName=MaxRoughness}
    Alpha=@n15
   )
   Normal=LinearInterpolate(
    A=@n29.RGB
    B=MaterialFunctionCall(Normal=@n6, Flatness=@n30) {materialFunction=&FlattenNormal}.Result
    Alpha=@n13.G
   ) {constB=0}
   EmissiveColor=ShadingPathSwitch(Default=QualitySwitch(Default=@n17, High=@n18, Epic=@n18), Mobile=Divide(A=@n22, B=@n23.EyeAdaptation))
   SubsurfaceColor=ShadingPathSwitch(Default=Multiply(A=@n24, B=@n25), Mobile=@n20)
   Opacity=ShadingPathSwitch(Default=@n28, Mobile=Constant() {r=1})
   ShadingModel=ShadingModel() {shadingModel=MSM_TwoSidedFoliage}
  ) {attributeSetTypes [
    #1
    #2
    #3
    #4
    #5
    #6
    #7
    #8
   ]}
  n32 = FunctionInput() {inputName=Material, inputType=FunctionInput_MaterialAttributes}
  FunctionOutput(None=StaticSwitchParameter(True=@n31, False=GetMaterialAttributes(None=@n32).MaterialAttributes) {parameterName=UseUndergoundSkylight})
  LinearInterpolate(A=@n27, B=@n26, Alpha=NamedRerouteUsage())
  MaterialFunctionCall() {materialFunction=&CameraDirectionVector}
  Multiply(A=@n3.RGB, B=Constant3Vector() {constant=<rgba 1 -1 1 0>}) {constB=-1}
  Multiply()
  Multiply(A=@n28)
  n33 = SphereMask(
   A=TextureCoordinate()
   B=Constant2Vector() {g=1, r=1}
   Radius=ScalarParameter() {defaultValue=0.5, parameterName=SpheremaskRadius}
   Hardness=ScalarParameter() {parameterName=SpheremaskHardnes}
  ) {attenuationRadius=0.5, hardnessPercent=0}
  NamedRerouteDeclaration(None=Saturate(None=Power(Base=@n33) {constExponent=4.3673})) {name=SphereMask, nodeColor=<rgba 0 1 0.625 1>}
  NamedRerouteDeclaration(None=LinearInterpolate(A=ScalarParameter() {parameterName=MinOcclusion}, B=ScalarParameter() {defaultValue=1, parameterName=MaxOcclusion}, Alpha=@n14) {constAlpha=1}) {name=Occlusion, nodeColor=<rgba 1 0.9375 0 1>}
  n34 = ReflectionVectorWS(CustomWorldNormal=@n7)
  n35 = Multiply(A=Multiply(A=NamedRerouteUsage(), B=SkyLightEnvMapSample(Direction=@n34)), B=Saturate(None=@n8))
  n36 = ScalarParameter() {defaultValue=1, group=PostRVTSnow, parameterName=Epic_SnowSparkleEmissiveBrightness}
  NamedRerouteDeclaration(None=Multiply(A=@n13.G, B=Multiply(A=@n35, B=@n36))) {name=ReflectionEmissive, nodeColor=<rgba 0.4531 0 1 1>}
  NamedRerouteDeclaration(None=@n20) {name=BaseColor, nodeColor=<rgba 0.1484 1 0 1>}
  NamedRerouteUsage()
  PixelDepth()
  PixelNormalWS()
  Saturate(None=@n16)
  ScalarParameter() {defaultValue=0.1, group=PostRVTSnow, parameterName=SparkleMin}
 ]}
